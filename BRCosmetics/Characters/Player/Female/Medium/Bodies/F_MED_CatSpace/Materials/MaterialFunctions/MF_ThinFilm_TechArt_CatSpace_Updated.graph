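# Material function for the CatSpace character body that adds two masked thin-film iridescence layers to BaseColor.
# Each layer samples a spectrum texture along a warped, scaled, offset Fresnel coordinate, then adds it to BaseColor weighted by a Fresnel power.
# Selectable channels of the Alpha input mask each layer, with the second blended over the first.
# The first layer's intensity is an input to allow different front and back strengths; the second uses a parameter.
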MaterialFunction /BRCosmetics/Characters/Player/Female/Medium/Bodies/F_MED_CatSpace/Materials/MaterialFunctions/MF_ThinFilm_TechArt_CatSpace_Updated
graph {nodes [
  n1 = NamedRerouteUsage()
  n2 = NamedRerouteUsage()
  n3 = ScalarParameter() {defaultValue=1, group=06_Thin Film, parameterName=ThinFilmScale, sortPriority=5}
  n4 = Power(Base=NamedRerouteUsage(), Exp=ScalarParameter() {defaultValue=0.5, group=06_Thin Film, parameterName=ThinFilmWarp, sortPriority=4})
  n5 = TextureSampleParameter2D(UVs=Add(A=Multiply(A=@n3, B=@n4), B=ScalarParameter() {group=06_Thin Film, parameterName=ThinFilmOffset})) {
   group=06_Thin Film
   parameterName=Thin Film Texture
   sortPriority=1
   texture=/Game/Global/Textures/Default/Thinfilm/T_ThinFilm_Spectrum_COLOR
  }
  n6 = Multiply(A=@n5.RGB, B=NamedRerouteUsage())
  n7 = Power(Base=NamedRerouteUsage(), Exp=ScalarParameter() {defaultValue=2, group=06_Thin Film, parameterName=ThinFilmExponent, sortPriority=3})
  n8 = SetMaterialAttributes!(MaterialAttributes=NamedRerouteUsage(), BaseColor=LinearInterpolate(A=@n2, B=Add(A=@n2, B=@n6), Alpha=Saturate(None=@n7))) {attributeSetTypes [#1]}
  n9 = ChannelMaskParameter(None=NamedRerouteUsage()) {group=05_Tech Art Masks, parameterName=ThinFilmMaskChannel}
  n10 = NamedRerouteUsage()
  n11 = NamedRerouteUsage()
  n12 = ScalarParameter() {defaultValue=1, group=06_Thin Film_2, parameterName=ThinFilmScale2, sortPriority=5}
  n13 = Power(Base=NamedRerouteUsage(), Exp=ScalarParameter() {defaultValue=0.5, group=06_Thin Film_2, parameterName=ThinFilmWarp2, sortPriority=4})
  n14 = TextureSampleParameter2D(UVs=Add(A=Multiply(A=@n12, B=@n13), B=ScalarParameter() {group=06_Thin Film_2, parameterName=ThinFilmOffset2})) {group=06_Thin Film, parameterName=ThinFilmTexture2, sortPriority=1, texture=&T_ThinFilm_Spectrum_COLOR}
  n15 = ScalarParameter() {defaultValue=0.5, group=06_Thin Film_2, parameterName=ThinFilmIntensity2, sortPriority=2}
  n16 = NamedRerouteUsage()
  n17 = ScalarParameter() {defaultValue=2, group=06_Thin Film_2, parameterName=ThinFilmExponent2, sortPriority=3}
  n18 = LinearInterpolate(A=@n11, B=Add(A=@n11, B=Multiply(A=@n14.RGB, B=@n15)), Alpha=Saturate(None=Power(Base=@n16, Exp=@n17)))
  n19 = NamedRerouteUsage()
  FunctionOutput(None=BlendMaterialAttributes(
    A=BlendMaterialAttributes(A=@n1, B=@n8, Alpha=@n9.R) {vertexAttributeBlendType=UseA}
    B=SetMaterialAttributes(MaterialAttributes=@n10, BaseColor=@n18) {attributeSetTypes [#1]}
    Alpha=ChannelMaskParameter(None=@n19) {group=05_Tech Art Masks, parameterName=ThinFilm2MaskChannel}.R
   ) {vertexAttributeBlendType=UseA})
  n20 = GetMaterialAttributes(None=FunctionInput() {inputName=Material Attributes, inputType=FunctionInput_MaterialAttributes, sortPriority=1}) {attributeGetTypes [#1]}
  NamedRerouteDeclaration(None=@n20.MaterialAttributes) {name=Material Attributes, nodeColor=<rgba 0.0781 0 1 1>}
  n21 = FunctionInput() {inputName=Fresnel, inputType=FunctionInput_Scalar, sortPriority=2}
  NamedRerouteDeclaration(None=@n21) {name=Fresnel, nodeColor=<rgba 0 0.8961 1 1>}
  NamedRerouteDeclaration(None=@n20.BaseColor) {name=BaseColor, nodeColor=<rgba 1 0 0.6562 1>}
  n22 = FunctionInput() {bUsePreviewValueAsDefault=true, inputName=Alpha, sortPriority=3}
  NamedRerouteDeclaration(None=@n22) {name=Alpha, nodeColor=<rgba 0.9516 0.9439 1 1>}
  n23 = FunctionInput() {
   desc="We made that an Input so we could have different Thin\r\nFilm Intensities on the Front and Back of the Character."
   inputName=CustomThinFilmIntensity
   inputType=FunctionInput_Scalar
   previewValue=<0.5 0 0 1>
  }
  NamedRerouteDeclaration(None=@n23) {name=CustomThinFilmIntensity, nodeColor=<rgba 0.2422 1 0 1>}
  Reroute(None=@n21)
  Reroute(None=@n20.BaseColor)
  Reroute(None=@n20.BaseColor)
  Reroute(None=@n22)
  Reroute(None=@n23)
 ]}
